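# Translucent unlit material for Niagara mesh particles that renders a glowing, tinted skull ghost.
# The red channel of the skull texture is tinted by Color1 and combined with Fresnel rim terms, then multiplied by particle color and passed through shared FX emissive adjustments.
# An alpha chain multiplies particle alpha by a contrast-adjusted near-camera fade and a local-Z height gradient that vertex-color red blends in.
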Material /BRCosmetics/Effects/Niagara/MasterMaterials/Mesh/M_SkullGhost
graph {
 properties {
  bAutomaticallySetUsageInEditor=false
  bUsedWithNiagaraMeshParticles=true
  blendMode=BLEND_Translucent
  shadingModel=MSM_Unlit
  usedShadingModels=Unlit
 }
 nodes [
  n1 = ParticleColor()
  n2 = VectorParameter() {parameterName=Color1, defaultValue=<rgba 0 1 0.3041 1>}
  n3 = Multiply(A=TextureSample() {texture=/BRCosmetics/Effects/Niagara/MasterMaterials/Mesh/Texture/T_SkullGhost}.R, B=@n2.RGB)
  n4 = Fresnel() {baseReflectFraction=0.2, exponent=3}
  n5 = Add(
   A=Multiply(A=Fresnel() {baseReflectFraction=0, exponent=3}, B=Multiply(A=@n2.RGB))
   B=Multiply(B=LinearInterpolate(B=@n3, Alpha=@n4) {constA=0.1}) {constA=1}
  )
  MaterialFunctionCall(Emissive=Multiply(A=@n1.RGB, B=Multiply(A=@n5))) {materialFunction=&MF_FX_EmissiveAdjustments}
  n6 = MaterialFunctionCall(Distance=ScalarParameter() {defaultValue=300, parameterName=NearCameraFade}) {materialFunction=&Near_Camera_Fade}
  n7 = ScalarParameter() {defaultValue=0.35, parameterName=Contrast}
  n8 = VertexColor()
  n9 = LocalPosition() {includedOffsets=ExcludeOffsets}
  n10 = Add(A=Divide(A=ComponentMask(None=@n9.XYZ) {b=true}) {constB=10}, B=ScalarParameter() {defaultValue=1, parameterName=HeightMaskOffset})
  Multiply(
   A=Multiply(
    A=@n1.A
    B=MaterialFunctionCall(In=@n6.Result, Contrast=@n7) {materialFunction=&CheapContrast}.Result
   )
   B=LinearInterpolate(B=Multiply(A=@n8.R, B=@n10), Alpha=@n8.R) {constA=1}
  )
 ]
}
