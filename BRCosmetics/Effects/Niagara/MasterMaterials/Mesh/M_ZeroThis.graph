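# M_ZeroThis is an unlit master material for skeletal meshes and mesh deformers in the BRCosmetics Niagara effects set.
# Its only logic multiplies world position, excluding shader offsets, by -1.
# The name suggests this negated position serves as an offset that collapses vertices to the world origin to hide a mesh, though the graph shows no output connection.
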Material /BRCosmetics/Effects/Niagara/MasterMaterials/Mesh/M_ZeroThis
graph {
 properties {bUsedWithMeshDeformer=true, bUsedWithSkeletalMesh=true, pixelDepthOffsetMode=PDOM_Legacy, shadingModel=MSM_Unlit, usedShadingModels=Unlit}
 nodes [Multiply(A=WorldPosition() {worldPositionShaderOffset=WPT_ExcludeAllShaderOffsets}.XYZ) {constB=-1}]
}
