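# Material function for FX master materials that derives a simple normal from a scalar Height mask, previewed with an exponential radial gradient.
# Screen-space DDX and DDY of Height are each multiplied by Height, combined into a two-channel slope and scaled by Strength.
# Outputs that slope as RG, or as RGB with a constant Z of one and no normalization.
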MaterialFunction /BRCosmetics/Effects/Niagara/MasterMaterials/Functions/MF_FX_SimpleNormalFromMask
graph {nodes [
  n1 = FunctionInput(Preview=MaterialFunctionCall() {materialFunction=&RadialGradientExponential}.RadialGradientExponential) {bUsePreviewValueAsDefault=true, inputName=Height, inputType=FunctionInput_Scalar}
  n2 = DDX(Value=@n1)
  n3 = DDY(Value=@n1)
  n4 = Multiply(
   A=AppendVector(A=DotProduct(A=@n2, B=@n1), B=DotProduct(A=@n1, B=@n3))
   B=FunctionInput(Preview=Constant() {r=10}) {bUsePreviewValueAsDefault=true, inputName=Strength, inputType=FunctionInput_Scalar, sortPriority=1}
  )
  FunctionOutput(None=@n4) {outputName=RG, sortPriority=1}
  FunctionOutput(None=AppendVector(A=@n4, B=Constant() {r=1})) {outputName=RGB}
 ]}
